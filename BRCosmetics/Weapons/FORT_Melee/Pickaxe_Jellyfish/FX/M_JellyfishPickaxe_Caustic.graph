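# Translucent unlit particle-sprite material that draws animated caustics for the Jellyfish pickaxe effect.
# Two panning distortion maps warp caustic UVs, which tint between DarkColor and particle color, plus a noise-pulsed gradient highlight.
# A scene-depth and curve-remapped ambient-occlusion mask drives bump-offset parallax and, with sphere mask and depth fades, opacity.
# Mobile skips parallax and depth masking for a simpler caustic sphere mask; TestView displays opacity as emissive.
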Material /BRCosmetics/Weapons/FORT_Melee/Pickaxe_Jellyfish/FX/M_JellyfishPickaxe_Caustic
graph {
 properties {
  bUsedWithParticleSprites=true
  blendMode=BLEND_Translucent
  pixelDepthOffsetMode=PDOM_Legacy
  shadingModel=MSM_Unlit
  twoSided=true
  usedShadingModels=Unlit
 }
 nodes [
  n1 = ObjectPositionWS()
  n2 = DynamicParameter() {paramNames ["UVTweakedByPos(0-1)", "SphereMaskRadius(0-0.5)", "SphereMaskHardness(0-1)", Param4]}
  n3 = VectorParameter() {group=Caustic, parameterName=CausticUV_tile, sortPriority=1, defaultValue=<rgba 1 1 0 1>}
  n4 = Add(
   A=ComponentMask(None=Multiply(A=@n1, B=@n2."UVTweakedByPos(0-1)")) {g=true, r=true}
   B=Multiply(A=TextureCoordinate(), B=ComponentMask(None=@n3.RGB) {g=true, r=true})
  )
  n5 = Subtract(A=ParticlePositionWS(), B=CameraPositionWS())
  n6 = Add(
   A=Abs(None=MaterialFunctionCall(Vector 3=@n5) {materialFunction=/Engine/Functions/Engine_MaterialFunctions02/Utility/VectorLength}.V3 Length)
   B=ScalarParameter() {defaultValue=40, group=DepthMask, parameterName=DethLength}
  ) {constB=40}
  n7 = Clamp(None=Multiply(A=SceneTexture() {sceneTextureId=PPI_SceneDepth}.Color, B=Divide(B=@n6)) {constB=0.0043})
  n8 = CurveAtlasRowParameter(CurveTime=ComponentMask(None=SceneTexture() {sceneTextureId=PPI_AmbientOcclusion}.Color) {r=true}) {
   atlas=&AthenaCurveAtlas01
   curve=/Game/Athena/LinearColorCurves/Curve_JellyfishPickaxe_caustic
   defaultValue=99
   group=DepthMask
   sortPriority=1
  }
  n9 = Multiply(A=ComponentMask(None=OneMinus(None=@n7)) {r=true}, B=@n8.G)
  n10 = BumpOffset(Coordinate=@n4, Height=Constant() {r=3}, HeightRatioInput=LinearInterpolate(Alpha=@n9) {constB=0.3})
  n11 = VectorParameter() {group=Caustic, parameterName=DistortMap1_UVpanSpeed, sortPriority=5, defaultValue=<rgba 0.1 0.1 0 1>}
  n12 = ComponentMask(None=@n11.RGB) {g=true, r=true}
  n13 = VectorParameter() {group=Caustic, parameterName=DistortMap1_UVtile, sortPriority=4, defaultValue=<rgba 1 1 0 1>}
  n14 = TextureSampleParameter2D(UVs=Multiply(A=Panner(Speed=@n12) {speedX=0.1, speedY=0.1}, B=ComponentMask(None=@n13.RGB) {g=true, r=true})) {
   group=Caustic
   parameterName=DistortMap1
   samplerType=SAMPLERTYPE_Normal
   sortPriority=3
   texture=/Game/Effects/Fort_Effects/Textures/T_DustDistortion
  }
  n15 = Multiply(A=@n14.RGB, B=ScalarParameter() {defaultValue=0.1, group=Caustic, parameterName=DistortMap1_Strength, sortPriority=9}) {constB=0.1}
  n16 = VectorParameter() {group=Caustic, parameterName=DistortMap2_UVpanSpeed, sortPriority=8, defaultValue=<rgba -0.09 -0.07 0 1>}
  n17 = Panner(Speed=ComponentMask(None=@n16.RGB) {g=true, r=true}) {speedX=-0.09, speedY=-0.07}
  n18 = ComponentMask(None=VectorParameter() {group=Caustic, parameterName=DistortMap2_UVtile, sortPriority=7, defaultValue=<rgba 1 1 0 1>}.RGB) {g=true, r=true}
  n19 = Multiply(
   A=TextureSampleParameter2D(UVs=Multiply(A=@n17, B=@n18)) {group=Caustic, parameterName=DistortMap2, sortPriority=6, texture=&T_NoisePack_01}.RGB
   B=ScalarParameter() {defaultValue=0.1, group=Caustic, parameterName=DistortMap2_Strength, sortPriority=10}
  ) {constB=0.1}
  n20 = Add(
   A=Multiply(A=ShadingPathSwitch(Default=@n10, Mobile=@n4) {desc=Migrated from FeatureLevelSwitch}, B=Constant2Vector() {g=0.5, r=0.5})
   B=StaticComponentMaskParameter(None=Add(A=@n15, B=@n19)) {defaultG=true, defaultR=true, group=Caustic, parameterName=CausticDistortChannel, sortPriority=2}
  )
  n21 = TextureSampleParameter2D(UVs=@n20) {group=Caustic, parameterName=Caustic_texture, texture=/Game/Global/Textures/Default/Caustics/T_Caustics_Softer}
  n22 = LinearInterpolate(A=VectorParameter() {parameterName=DarkColor, defaultValue=<rgba 0.2521 0.6439 0.8177 1>}.RGB, B=ParticleColor().RGB, Alpha=@n21.R)
  n23 = Sine(None=Multiply(A=Time()) {constB=0.5})
  n24 = TextureSampleParameter2D(UVs=Add(A=@n20, B=Multiply(A=@n23) {constB=0.012})) {group=Caustic, parameterName=Caustic_texture, texture=&T_Caustics_Softer}
  n25 = Multiply(A=Time()) {constB=0.3}
  n26 = Divide(A=Add(A=Sine(None=@n25)))
  n27 = OneMinus(None=@n26)
  n28 = TextureSample() {samplerType=SAMPLERTYPE_LinearColor, texture=&HighContrastNoise}
  n29 = Multiply(
   A=TextureSample(UVs=Multiply(A=@n24.R) {constB=0.5}) {texture=/Game/Effects/Fort_Effects/Textures/Grad/T_Halo_Corona_Gradient}.RGB
   B=Multiply(A=LinearInterpolate(A=@n27, B=@n26, Alpha=@n28.R)) {constB=5}
  ) {constB=2}
  n30 = SphereMask(
   A=TextureCoordinate()
   B=Constant() {r=0.5}
   Radius=ScalarParameter() {defaultValue=0.4, parameterName=SphereRadius}
   Hardness=ScalarParameter() {parameterName=SphereHardness}
  ) {attenuationRadius=0.5, hardnessPercent=0}
  n31 = DynamicParameter() {paramNames ["UVTweakedByPos(0-1)", "SphereMaskRadius(0-0.5)", "SphereMaskHardness(0-1)", Param4]}
  n32 = SphereMask(A=TextureCoordinate(), B=Constant2Vector() {g=0.5, r=0.5}, Radius=@n31."SphereMaskRadius(0-0.5)", Hardness=@n31."SphereMaskRadius(0-0.5)") {attenuationRadius=0.25, hardnessPercent=0}
  n33 = ShadingPathSwitch(Default=DepthFade(Opacity=Multiply(A=@n9, B=@n30)) {fadeDistanceDefault=10}, Mobile=Multiply(A=Multiply(A=@n32, B=@n21.R)) {constB=0.25}) {desc=Migrated from FeatureLevelSwitch}
  n34 = TextureCoordinate()
  n35 = MaterialFunctionCall(Float2=Sine(None=Multiply(A=@n34) {constB=0.5})) {materialFunction=&BreakOutFloat2Components}
  n36 = Multiply(A=@n35.R, B=@n35.R)
  n37 = Constant() {r=300}
  n38 = Multiply(
   A=Multiply(
    A=Multiply(A=@n33, B=@n36)
    B=MaterialFunctionCall(Fade Length=@n37) {materialFunction=&CameraDepthFade}.Result
   )
   B=ParticleColor().A
  )
  MaterialFunctionCall(Emissive=StaticSwitchParameter(True=@n38, False=Add(A=@n22, B=@n29)) {parameterName=TestView}) {materialFunction=&MF_FX_EmissiveAdjustments}
  StaticSwitchParameter(True=Constant() {r=1}, False=@n38) {parameterName=TestView}
 ]
}
