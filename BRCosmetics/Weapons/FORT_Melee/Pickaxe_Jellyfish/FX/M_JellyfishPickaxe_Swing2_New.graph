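# Unlit modulate-blend beam trail material for the swing effect of the Jellyfish pickaxe.
# Curve atlas rows sampled along the UVs shape a tinted color gradient, an animated caustic layer and a boosted panning noise layer.
# Dust and noise textures distort the caustics, with a CausticSpeed dynamic parameter driving their panning.
# A squared sine end mask, particle alpha and depth fade blend toward white, the neutral value for modulate.
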Material /BRCosmetics/Weapons/FORT_Melee/Pickaxe_Jellyfish/FX/M_JellyfishPickaxe_Swing2_New
graph {
 properties {
  bUsedWithBeamTrails=true
  blendMode=BLEND_Modulate
  pixelDepthOffsetMode=PDOM_Legacy
  shadingModel=MSM_Unlit
  translucencyPass=MTP_BeforeDOF
  usedShadingModels=Unlit
 }
 nodes [
  n1 = TextureCoordinate()
  n2 = MaterialFunctionCall(Float2=Sine(None=Multiply(A=@n1) {constB=0.5})) {materialFunction=&BreakOutFloat2Components}
  n3 = Multiply(A=@n2.R, B=@n2.R)
  n4 = CurveAtlasRowParameter(CurveTime=ComponentMask(None=TextureCoordinate()) {r=true}) {
   atlas=&AthenaCurveAtlas01
   curve=/Game/Athena/LinearColorCurves/Curve_JellyfishPickaxe_Swing2
   defaultValue=100
  }
  n5 = Multiply(A=TextureCoordinate(), B=Constant2Vector() {g=1.5, r=8})
  n6 = Multiply(A=TextureCoordinate(), B=Constant2Vector() {g=0.5, r=4})
  n7 = Multiply(
   A=TextureSample(UVs=Panner(Coordinate=@n5) {speedX=0.5}) {texture=/Game/Global/Textures/Default/Caustics/T_Turbulence_Seamless}.R
   B=TextureSample(UVs=Panner(Coordinate=@n6) {speedX=0.4}) {texture=&AtmosphericCloudNoise01}.R
  )
  n8 = TextureCoordinate()
  n9 = CurveAtlasRowParameter(CurveTime=ComponentMask(None=@n8) {g=true}) {
   atlas=&AthenaCurveAtlas01
   curve=&Curve_JellyfishPickaxe_Swing2
   defaultValue=100
  }
  n10 = OneMinus(None=ComponentMask(None=@n8) {r=true})
  n11 = Add(A=Multiply(A=@n9.A, B=Power(Base=@n10)))
  n12 = VectorParameter() {parameterName=DarkColor, defaultValue=<rgba 3 2.7234 1.4713 1>}
  n13 = VectorParameter() {parameterName=LightColor, defaultValue=<rgba 0.3047 1.5 1.338 1>}
  n14 = ComponentMask(None=TextureCoordinate()) {r=true}
  n15 = CurveAtlasRowParameter(CurveTime=@n14) {
   atlas=&AthenaCurveAtlas01
   curve=/Game/Athena/LinearColorCurves/Curve_JellyfishPickaxe_caustic
   defaultValue=99
  }
  n16 = CurveAtlasRowParameter(CurveTime=@n14) {
   atlas=&AthenaCurveAtlas01
   curve=&Curve_JellyfishPickaxe_Swing2
   defaultValue=100
  }
  n17 = Constant3Vector()
  n18 = Constant3Vector() {constant=<rgba 0.9741 1.7793 2 0>}
  n19 = Multiply(A=TextureCoordinate(), B=Constant2Vector() {g=0.5, r=1})
  n20 = DynamicParameter() {paramNames [CausticSpeed,Param2,Param3,Param4]}
  n21 = Multiply(A=@n20.CausticSpeed) {constB=0.1}
  n22 = TextureSample(UVs=Multiply(A=Panner(Speed=@n21) {speedX=0.1, speedY=0.1}, B=Constant2Vector() {g=0.5, r=1.5})) {samplerType=SAMPLERTYPE_Normal, texture=/Game/Effects/Fort_Effects/Textures/T_DustDistortion}
  n23 = Multiply(A=@n20.CausticSpeed) {constB=-0.07}
  n24 = TextureSample(UVs=Multiply(A=Panner(Speed=@n23) {speedX=-0.09, speedY=-0.07}, B=Constant2Vector() {g=0.5, r=1.5})) {texture=&T_NoisePack_01}
  n25 = Clamp(None=Add(A=Multiply(A=@n22.RGB) {constB=0.2}, B=Multiply(A=@n24.RGB) {constB=0.1}))
  n26 = TextureSample(UVs=Add(A=Panner(Coordinate=@n19) {speedX=0.2}, B=ComponentMask(None=@n25) {g=true, r=true})) {texture=/Game/Global/Textures/Default/Caustics/T_Caustics_Softer}
  n27 = CurveAtlasRowParameter(CurveTime=@n26.R) {
   atlas=&AthenaCurveAtlas01
   curve=&Curve_JellyfishPickaxe_caustic
   defaultValue=99
  }
  n28 = Add(A=Multiply(A=@n26.R) {constB=0.5}, B=Multiply(A=@n27.R) {constB=12}) {constB=0}
  n29 = CurveAtlasRowParameter(CurveTime=ComponentMask(None=TextureCoordinate()) {r=true}) {
   atlas=&AthenaCurveAtlas01
   curve=&Curve_JellyfishPickaxe_Swing2
   defaultValue=100
  }
  n30 = Add(
   A=LinearInterpolate(
    A=LinearInterpolate(A=@n12.RGB, B=@n13.RGB, Alpha=@n15.A)
    B=VectorParameter() {parameterName=DarkColorBlue, defaultValue=<rgba 0 0.7941 1 1>}.RGB
    Alpha=@n16.R
   )
   B=Multiply(A=LinearInterpolate(A=@n17, B=@n18, Alpha=@n28), B=@n29.G)
  )
  n31 = Add(A=Multiply(A=Multiply(A=@n11, B=@n30), B=ParticleColor().RGB), B=Multiply(A=Multiply(A=@n4.B, B=@n7)) {constB=70})
  n32 = LinearInterpolate(A=Constant() {r=1}, B=Add(A=OneMinus(None=@n3), B=Multiply(A=@n3, B=@n31)), Alpha=ParticleColor().A)
  MaterialFunctionCall(Emissive=LinearInterpolate(A=Constant() {r=1}, B=@n32, Alpha=DepthFade(Opacity=@n32) {fadeDistanceDefault=40})) {materialFunction=&MF_FX_EmissiveAdjustments}
 ]
}
