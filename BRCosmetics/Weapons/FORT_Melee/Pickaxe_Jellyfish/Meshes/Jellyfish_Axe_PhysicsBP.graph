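# Physics asset for the Jellyfish pickaxe cosmetic, defining bodies for dangling upper and lower leaf pieces on the weapon mesh.
# Two capsules on b_AR_Root cover the weapon, and small capsules cover each dyn_ leaf bone.
# Constraints link leaf chains to the root with 15-degree swing and twist limits, with lowerleafA branching after its first bone.
# lowerleafB_01 also carries a second root constraint with free angular motion.
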PhysicsAsset /BRCosmetics/Weapons/FORT_Melee/Pickaxe_Jellyfish/Meshes/Jellyfish_Axe_PhysicsBP
bodies {
 b_AR_Root [
  {shapeType=Capsule, length=105.1369, radius=2.75, center=<0.6988 7.1924 -13.3089>, rotation=<rot -1.3373 -90 -0.4058>}
  {shapeType=Capsule, length=26, radius=6, center=<0 7.502 57.591>}
 ]
 dyn_lowerleafA_01 [{shapeType=Capsule, length=2.0178, radius=2.0549, center=<0.2 0 0>, rotation=<rot 85.0016 -179.9998 -179.9998>}]
 dyn_lowerleafA_03 [{shapeType=Capsule, length=2.0178, radius=2.0549, center=<0.2 0 0>, rotation=<rot 85.0016 -179.9998 -179.9998>}]
 dyn_lowerleafA_04 [{shapeType=Capsule, length=2.0178, radius=2.0549, center=<0.2 0 0>, rotation=<rot 85.0016 -179.9998 -179.9998>}]
 dyn_lowerleafB_01 [{shapeType=Capsule, length=2.0178, radius=2.0549, center=<0.2 0 0>, rotation=<rot 85.0016 -179.9998 -179.9998>}]
 dyn_lowerleafB_02 [{shapeType=Capsule, length=2.0178, radius=2.0549, center=<0.2 0 0>, rotation=<rot 85.0016 -179.9998 -179.9998>}]
 dyn_lowerleafC_01 [{shapeType=Capsule, length=2.0178, radius=2.0549, center=<0.2 0 0>, rotation=<rot 85.0016 -179.9998 -179.9998>}]
 dyn_lowerleafC_02 [{shapeType=Capsule, length=2.0178, radius=2.0549, center=<0.2 0 0>, rotation=<rot 85.0016 -179.9998 -179.9998>}]
 dyn_lowerleafD_01 [{shapeType=Capsule, length=2.0178, radius=2.0549, center=<0.2 0 0>, rotation=<rot 85.0016 -179.9998 -179.9998>}]
 dyn_lowerleafD_02 [{shapeType=Capsule, length=2.0178, radius=2.0549, center=<0.2 0 0>, rotation=<rot 85.0016 -179.9998 -179.9998>}]
 dyn_lowerleafE_01 [{shapeType=Capsule, length=2.0178, radius=2.0549, center=<0.2 0 0>, rotation=<rot 85.0016 -179.9998 -179.9998>}]
 dyn_lowerleafE_02 [{shapeType=Capsule, length=2.0178, radius=2.0549, center=<0.2 0 0>, rotation=<rot 85.0016 -179.9998 -179.9998>}]
 dyn_lowerleafF_01 [{shapeType=Capsule, length=2.0178, radius=2.0549, center=<0.2 0 0>, rotation=<rot 85.0016 -179.9998 -179.9998>}]
 dyn_lowerleafF_02 [{shapeType=Capsule, length=2.0178, radius=2.0549, center=<0.2 0 0>, rotation=<rot 85.0016 -179.9998 -179.9998>}]
 dyn_lowerleafF_03 [{shapeType=Capsule, length=2.0178, radius=2.0549, center=<0.2 0 0>, rotation=<rot 85.0016 -179.9998 -179.9998>}]
 dyn_lowerleafG_01 [{shapeType=Capsule, length=2.0178, radius=2.0549, center=<0.2 0 0>, rotation=<rot 85.0016 -179.9998 -179.9998>}]
 dyn_lowerleafG_02 [{shapeType=Capsule, length=2.0178, radius=2.0549, center=<0.2 0 0>, rotation=<rot 85.0016 -179.9998 -179.9998>}]
 dyn_lowerleafH_01 [{shapeType=Capsule, length=2.0178, radius=2.0549, center=<0.2 0 0>, rotation=<rot 85.0016 -179.9998 -179.9998>}]
 dyn_lowerleafH_02 [{shapeType=Capsule, length=2.0178, radius=2.0549, center=<0.2 0 0>, rotation=<rot 85.0016 -179.9998 -179.9998>}]
 dyn_upperleafA_01 [{shapeType=Capsule, length=2.0178, radius=2.0549, center=<0.2 0 0>, rotation=<rot 85.0016 -179.9998 -179.9998>}]
 dyn_upperleafA_02 [{shapeType=Capsule, length=2.0178, radius=2.0549, center=<0.2 0 0>, rotation=<rot 85.0016 -179.9998 -179.9998>}]
 dyn_upperleafB_01 [{shapeType=Capsule, length=2.0178, radius=2.0549, center=<0.2 0 0>, rotation=<rot 85.0016 -179.9998 -179.9998>}]
 dyn_upperleafB_02 [{shapeType=Capsule, length=2.0178, radius=2.0549, center=<0.2 0 0>, rotation=<rot 85.0016 -179.9998 -179.9998>}]
 dyn_upperleafC_01 [{shapeType=Capsule, length=1.3126, radius=1.8743, center=<1.3895 0.1572 -0.0358>, rotation=<rot -1.3195 -88.9141 -90.2324>}]
 dyn_upperleafC_02 [{shapeType=Capsule, length=1.3126, radius=1.8743, center=<1.3895 0.1572 -0.0358>, rotation=<rot -1.3195 -88.9141 -90.2324>}]
 dyn_upperleafD_01 [{shapeType=Capsule, length=2.0178, radius=2.0549, center=<0.2 0 0>, rotation=<rot 85.0016 -179.9998 -179.9998>}]
 dyn_upperleafD_02 [{shapeType=Capsule, length=2.0178, radius=2.0549, center=<0.2 0 0>, rotation=<rot 85.0016 -179.9998 -179.9998>}]
 dyn_upperleafD_03 [{shapeType=Capsule, length=2.0178, radius=2.0549, center=<0.2 0 0>, rotation=<rot 85.0016 -179.9998 -179.9998>}]
 dyn_upperleafE_01 [{shapeType=Capsule, length=2.0178, radius=2.0549, center=<0.2 0 0>, rotation=<rot 90 -0.0528 -0.0528>}]
 dyn_upperleafE_02 [{shapeType=Capsule, length=2.0178, radius=2.0549, center=<0.2 0 0>, rotation=<rot 90 -0.0528 -0.0528>}]
 dyn_upperleafE_03 [{shapeType=Capsule, length=2.0178, radius=2.0549, center=<0.2 0 0>, rotation=<rot 90 -0.0528 -0.0528>}]
 dyn_upperleafF_01 [{shapeType=Capsule, length=2.0178, radius=2.0549, center=<0.2 0 0>, rotation=<rot 85.0016 -179.9998 -179.9998>}]
 dyn_upperleafF_02 [{shapeType=Capsule, length=2.0178, radius=2.0549, center=<0.2 0 0>, rotation=<rot 85.0016 -179.9998 -179.9998>}]
 dyn_upperleafG_01 [{shapeType=Capsule, length=2.0178, radius=2.0549, center=<0.2 0 0>, rotation=<rot 85.0016 -179.9998 -179.9998>}]
 dyn_upperleafG_02 [{shapeType=Capsule, length=2.75, radius=2.3049, center=<0.01 0 0>, rotation=<rot 85.0016 -179.9998 -179.9998>}]
}
constraints [
 {bone1Name=dyn_upperleafA_01, bone2Name=b_AR_Root, swing1LimitDegrees=15, swing2LimitDegrees=15, twistLimitDegrees=15}
 {bone1Name=dyn_upperleafB_01, bone2Name=b_AR_Root, swing1LimitDegrees=15, swing2LimitDegrees=15, twistLimitDegrees=15}
 {bone1Name=dyn_upperleafC_01, bone2Name=b_AR_Root, swing1LimitDegrees=15, swing2LimitDegrees=15, twistLimitDegrees=15}
 {bone1Name=dyn_upperleafD_01, bone2Name=b_AR_Root, swing1LimitDegrees=15, swing2LimitDegrees=15, twistLimitDegrees=15}
 {bone1Name=dyn_upperleafE_01, bone2Name=b_AR_Root, swing1LimitDegrees=15, swing2LimitDegrees=15, twistLimitDegrees=15}
 {bone1Name=dyn_upperleafF_01, bone2Name=b_AR_Root, swing1LimitDegrees=15, swing2LimitDegrees=15, twistLimitDegrees=15}
 {bone1Name=dyn_upperleafG_01, bone2Name=b_AR_Root, swing1LimitDegrees=15, swing2LimitDegrees=15, twistLimitDegrees=15}
 {bone1Name=dyn_lowerleafB_01, bone2Name=b_AR_Root, swing1LimitDegrees=15, swing2LimitDegrees=15, twistLimitDegrees=15}
 {
  bone1Name=dyn_lowerleafB_01
  bone2Name=b_AR_Root
  swing1LimitDegrees=45
  swing1Motion=Free
  swing2LimitDegrees=45
  swing2Motion=Free
  twistLimitDegrees=45
  twistMotion=Free
 }
 {bone1Name=dyn_upperleafG_02, bone2Name=dyn_upperleafG_01, swing1LimitDegrees=15, swing2LimitDegrees=15, twistLimitDegrees=15}
 {bone1Name=dyn_upperleafD_02, bone2Name=dyn_upperleafD_01, swing1LimitDegrees=15, swing2LimitDegrees=15, twistLimitDegrees=15}
 {bone1Name=dyn_upperleafD_03, bone2Name=dyn_upperleafD_02, swing1LimitDegrees=15, swing2LimitDegrees=15, twistLimitDegrees=15}
 {bone1Name=dyn_upperleafF_02, bone2Name=dyn_upperleafF_01, swing1LimitDegrees=15, swing2LimitDegrees=15, twistLimitDegrees=15}
 {bone1Name=dyn_upperleafA_02, bone2Name=dyn_upperleafA_01, swing1LimitDegrees=15, swing2LimitDegrees=15, twistLimitDegrees=15}
 {bone1Name=dyn_upperleafC_02, bone2Name=dyn_upperleafC_01, swing1LimitDegrees=15, swing2LimitDegrees=15, twistLimitDegrees=15}
 {bone1Name=dyn_upperleafB_02, bone2Name=dyn_upperleafB_01, swing1LimitDegrees=15, swing2LimitDegrees=15, twistLimitDegrees=15}
 {bone1Name=dyn_upperleafE_02, bone2Name=dyn_upperleafE_01, swing1LimitDegrees=15, swing2LimitDegrees=15, twistLimitDegrees=15}
 {bone1Name=dyn_upperleafE_03, bone2Name=dyn_upperleafE_02, swing1LimitDegrees=15, swing2LimitDegrees=15, twistLimitDegrees=15}
 {bone1Name=dyn_lowerleafB_02, bone2Name=dyn_lowerleafB_01, swing1LimitDegrees=15, swing2LimitDegrees=15, twistLimitDegrees=15}
 {bone1Name=dyn_lowerleafA_01, bone2Name=b_AR_Root, swing1LimitDegrees=15, swing2LimitDegrees=15, twistLimitDegrees=15}
 {bone1Name=dyn_lowerleafA_03, bone2Name=dyn_lowerleafA_01, swing1LimitDegrees=15, swing2LimitDegrees=15, twistLimitDegrees=15}
 {bone1Name=dyn_lowerleafA_04, bone2Name=dyn_lowerleafA_01, swing1LimitDegrees=15, swing2LimitDegrees=15, twistLimitDegrees=15}
 {bone1Name=dyn_lowerleafC_01, bone2Name=b_AR_Root, swing1LimitDegrees=15, swing2LimitDegrees=15, twistLimitDegrees=15}
 {bone1Name=dyn_lowerleafC_02, bone2Name=dyn_lowerleafC_01, swing1LimitDegrees=15, swing2LimitDegrees=15, twistLimitDegrees=15}
 {bone1Name=dyn_lowerleafD_02, bone2Name=dyn_lowerleafD_01, swing1LimitDegrees=15, swing2LimitDegrees=15, twistLimitDegrees=15}
 {bone1Name=dyn_lowerleafD_01, bone2Name=b_AR_Root, swing1LimitDegrees=15, swing2LimitDegrees=15, twistLimitDegrees=15}
 {bone1Name=dyn_lowerleafE_01, bone2Name=b_AR_Root, swing1LimitDegrees=15, swing2LimitDegrees=15, twistLimitDegrees=15}
 {bone1Name=dyn_lowerleafE_02, bone2Name=dyn_lowerleafE_01, swing1LimitDegrees=15, swing2LimitDegrees=15, twistLimitDegrees=15}
 {bone1Name=dyn_lowerleafF_01, bone2Name=b_AR_Root, swing1LimitDegrees=15, swing2LimitDegrees=15, twistLimitDegrees=15}
 {bone1Name=dyn_lowerleafF_02, bone2Name=dyn_lowerleafF_01, swing1LimitDegrees=15, swing2LimitDegrees=15, twistLimitDegrees=15}
 {bone1Name=dyn_lowerleafF_03, bone2Name=dyn_lowerleafF_02, swing1LimitDegrees=15, swing2LimitDegrees=15, twistLimitDegrees=15}
 {bone1Name=dyn_lowerleafG_01, bone2Name=b_AR_Root, swing1LimitDegrees=15, swing2LimitDegrees=15, twistLimitDegrees=15}
 {bone1Name=dyn_lowerleafG_02, bone2Name=dyn_lowerleafG_01, swing1LimitDegrees=15, swing2LimitDegrees=15, twistLimitDegrees=15}
 {bone1Name=dyn_lowerleafH_01, bone2Name=b_AR_Root, swing1LimitDegrees=15, swing2LimitDegrees=15, twistLimitDegrees=15}
 {bone1Name=dyn_lowerleafH_02, bone2Name=dyn_lowerleafH_01, swing1LimitDegrees=15, swing2LimitDegrees=15, twistLimitDegrees=15}
]
